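# MetaSound patch for one direction of a glider thrust sound: a random loop plus threshold-triggered one-shots in stereo.
# An interpolated Value drives loop pitch and volume, with an optional random start time.
# bIsBackwards selects the threshold and comparison types at play.
# Two self-closing trigger gates re-arm each other as Value crosses back and forth, preventing repeated one-shots, which are picked without immediate repeats and reported through OnOneshotPlay.
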MetaSoundPatch /FortMetasound/Cosmetics/Glider/Glider_ThrustLoop_Direction
inputs [
 bIsBackwards {type=Bool}
 bUseRandomStartTime {type=Bool}
 Interp Time {type=Time}
 Loop {type=WaveAsset:Array}
 On Play {type=Trigger}
 Oneshots {type=WaveAsset:Array}
 OneShotThreshold {type=Float}
 OneShotVolume {type=Float}
 PitchMax {type=Float}
 PitchMin {type=Float}
 "ReverseOneShotThreshold " {type=Float}
 Value {type=Float}
 VolumeMax {type=Float}
 VolumeMin {type=Float}
]
interfaces [MetaSound:1.0]
outputs [
 OnOneshotPlay {type=Trigger}
 Out L {type=Audio}
 Out R {type=Audio}
]
pages [Default {
  input_defaults {
   Interp Time=0.5
   On Play=null
   OneShotThreshold=0.5
   OneShotVolume=1
   PitchMax=1.2
   PitchMin=0.8
   "ReverseOneShotThreshold "=0.2
   Value=0
   VolumeMax=1
   VolumeMin=0.2
   bIsBackwards=null
   bUseRandomStartTime="<MetaSoundBuilderResult.SUCCEEDED: 0>"
   Loop []
   Oneshots []
  }
  nodes [
   n1 = VariableAccessor.Float(Variable=VariableMutator.Float(Variable=InitVariable.Float().Variable, Value=UE.InterpTo:Audio(Interp Time=^Interp Time, Target=^Value).Value).Variable)
   ArrayRandomGet_1 = Array.Random Get:WaveAsset:Array(In Array=^Loop, Next=^On Play)
   FrequencyMultipliertoSemitone_0 = UE.Frequency Multiplier to Semitone:Float(Frequency Multiplier=MapRange.MapRange:Float(Out Range B=^PitchMax, Out Range A=^PitchMin, In=@n1.Value).Out Value)
   n2 = TriggerCompare.Clamp:Bool(A=^bUseRandomStartTime, Compare=@ArrayRandomGet_1.On Next) {also {defaults {B="<MetaSoundBuilderResult.SUCCEEDED: 0>"}}}
   TriggerRouteTriggerRouteTime2_0 = "TriggerRoute.Trigger Route (Time, 2):Time"(Value 0=&MSP_WaveGetRandomStartTime(Sound=@ArrayRandomGet_1.Value).Value, Set 1=@n2.FALSE, Set 0=@n2.TRUE)
   WavePlayer_0 = UE.Wave Player:Stereo(
    Wave Asset=@ArrayRandomGet_1.Value
    Pitch Shift=UE.Reroute(Value=@FrequencyMultipliertoSemitone_0.Semitones).Value
    Play=@TriggerRouteTriggerRouteTime2_0.On Set
    Start Time=@TriggerRouteTriggerRouteTime2_0.Value
   ) {also {defaults {Loop="<MetaSoundBuilderResult.SUCCEEDED: 0>"}, version [1,1]}}
   n3 = TriggerCompare.Clamp:Bool(Compare=^On Play, A=^bIsBackwards)
   n4 = VariableAccessor.Trigger(Variable=VariableMutator.Trigger(Variable=InitVariable.Trigger().Variable, Value=@n3.FALSE).Variable)
   n5 = VariableAccessor.Trigger(Variable=@n4.Variable)
   n6 = InitVariable.Trigger()
   n7 = InitVariable.Trigger()
   TriggerDelay_0 = UE.Trigger Delay(
    Delay Time=VariableAccessor.Time(Variable=InitVariable.Time().Variable).Value
    In=VariableDeferredAccessor.Trigger(Variable=@n7.Variable).Value
   )
   TriggerAnyTriggerAny2_2 = "TriggerAny.Trigger Any (2)"(
    In 1=VariableAccessor.Trigger(Variable=@n5.Variable).Value
    In 0=VariableAccessor.Trigger(Variable=VariableMutator.Trigger(Variable=@n6.Variable, Value=@TriggerDelay_0.Out).Variable).Value
   )
   TriggerAnyTriggerAny2_1 = "TriggerAny.Trigger Any (2)"(In 1=@n4.Value, In 0=VariableDeferredAccessor.Trigger(Variable=@n7.Variable).Value)
   TriggerRouteTriggerRouteInt322_0 = "TriggerRoute.Trigger Route (Int32, 2):Int32"(Set 1=@n3.FALSE, Set 0=@n3.TRUE) {also {defaults {Value 0=2, Value 1=3}}}
   TriggerOnValueChange_0 = UE.Trigger On Value Change:Float(Value=^Value)
   TriggerRouteTriggerRouteFloat2_0 = "TriggerRoute.Trigger Route (Float, 2):Float"(Value 0=^OneShotThreshold, Value 1=^"ReverseOneShotThreshold ", Set 1=@n3.FALSE, Set 0=@n3.TRUE)
   n8 = VariableAccessor.Float(Variable=VariableMutator.Float(Variable=InitVariable.Float().Variable, Value=@TriggerRouteTriggerRouteFloat2_0.Value).Variable)
   n9 = VariableDeferredAccessor.Trigger(Variable=InitVariable.Trigger().Variable)
   TriggerControl_0 = UE.Trigger Control(
    Open=@TriggerAnyTriggerAny2_1.Out
    Trigger In=TriggerCompare.Clamp:Float(
     A=^Value
     Type=Convert.Enum:TriggerComparisonType:Int32(Int32=@TriggerRouteTriggerRouteInt322_0.Value)."Enum:TriggerComparisonType"
     Compare=@TriggerOnValueChange_0.Trigger
     B=VariableAccessor.Float(Variable=@n8.Variable).Value
    ).TRUE
    Close=@n9.Value
   )
   TriggerAccumulatorTriggerAccumulate2_0 = "TriggerAccumulator.Trigger Accumulate (2)"(
    In 0=@TriggerAnyTriggerAny2_2.Out
    In 1=VariableAccessor.Trigger(Variable=VariableMutator.Trigger(Value=@TriggerControl_0.Trigger Out, Variable=@n9.Variable).Variable).Value
   ) {also {defaults {Auto Reset="<MetaSoundBuilderResult.SUCCEEDED: 0>"}}}
   n10 = VariableDeferredAccessor.Trigger(Variable=@n7.Variable)
   TriggerAnyTriggerAny2_0 = "TriggerAny.Trigger Any (2)"(In 1=@n5.Value, In 0=@n10.Value)
   TriggerRouteTriggerRouteInt322_1 = "TriggerRoute.Trigger Route (Int32, 2):Int32"(Set 1=@n3.FALSE, Set 0=@n3.TRUE) {also {defaults {Value 0=5, Value 1=4}}}
   TriggerOnValueChange_1 = UE.Trigger On Value Change:Float(Value=^Value)
   TriggerControl_1 = UE.Trigger Control(
    Open=@TriggerAccumulatorTriggerAccumulate2_0.Out
    Close=@TriggerAnyTriggerAny2_0.Out
    Trigger In=TriggerCompare.Clamp:Float(
     A=^Value
     Type=Convert.Enum:TriggerComparisonType:Int32(Int32=@TriggerRouteTriggerRouteInt322_1.Value)."Enum:TriggerComparisonType"
     Compare=@TriggerOnValueChange_1.Trigger
     B=@n8.Value
    ).TRUE
   )
   n11 = /FortMetasound/Patches/MSP_GetNoRepeatNumber(
    WaveArray=^Oneshots
    InTrigger=VariableAccessor.Trigger(Variable=VariableMutator.Trigger(Value=@TriggerControl_1.Trigger Out, Variable=@n10.Variable).Variable).Value
   )
   ArrayRandomGet_0 = Array.Random Get:WaveAsset:Array(In Array=^Oneshots, No Repeats=@n11.NoRepeatNumber, Next=@n11.OutTrigger)
   WavePlayer_1 = UE.Wave Player:Stereo(Play=@ArrayRandomGet_0.On Next, Wave Asset=@ArrayRandomGet_0.Value) {also {version [1,1]}}
   AudioMixerAudioMixerStereo2_0 = "AudioMixer.Audio Mixer (Stereo, 2)"(
    Gain 1=^OneShotVolume
    Gain 0=MapRange.MapRange:Float(Out Range B=^VolumeMax, Out Range A=^VolumeMin, In=VariableAccessor.Float(Variable=@n1.Variable).Value).Out Value
    In 0 L=@WavePlayer_0.Out Left
    In 0 R=@WavePlayer_0.Out Right
    In 1 L=@WavePlayer_1.Out Left
    In 1 R=@WavePlayer_1.Out Right
   )
   InitVariable.Bool()
   InitVariable.Trigger()
   TriggerDelay_1 = UE.Trigger Delay(In=^On Play, Delay Time=UE.ConversionFloatToTime().Out)
   VariableMutator.Trigger(Variable=InitVariable.Trigger().Variable, Value=@TriggerDelay_1.Out)
  ]
  links [
   {dst="@Out L", src=AudioMixerAudioMixerStereo2_0.Out L}
   {dst="@Out R", src=AudioMixerAudioMixerStereo2_0.Out R}
   {dst="@OnOneshotPlay", src=WavePlayer_1.On Play}
  ]
 }]
