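# One-shot stereo MetaSound source that serves as the glider open cosmetic sound, with a wave array input that is empty by default.
# It passes the waves, start delay, pitch and volume inputs to the shared MSS_Cosmetic_Simple_Base graph and outputs that graph's audio and finished trigger.
# It also writes an attenuated mono sum of both channels to the glider-open audio bus, filed under volume envelope followers.
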MetaSoundSource /FortMetasound/Cosmetics/Glider/MSS_Glider_Open
inputs [
 Open.Sounds {type=WaveAsset:Array}
 PitchModulationRange.Max {type=Float}
 PitchModulationRange.Min {type=Float}
 StartDelay.Max {type=Time}
 StartDelay.Min {type=Time}
 TemplatePitchMultiplier {type=Float}
 TemplateVolumeMultiplier {type=Float}
 UE.Source.OnPlay {type=Trigger}
 VolumeModulationRange.Max {type=Float}
 VolumeModulationRange.Min {type=Float}
]
interfaces [UE.OutputFormat.Stereo:1.0, UE.Source.OneShot:1.0, UE.Source:1.1]
outputs [
 "UE.OutputFormat.Stereo.Audio:0" {type=Audio}
 "UE.OutputFormat.Stereo.Audio:1" {type=Audio}
 UE.Source.OneShot.OnFinished {type=Trigger}
]
pages [Default {
  input_defaults {
   PitchModulationRange.Max=1
   PitchModulationRange.Min=1
   StartDelay.Max=0
   StartDelay.Min=0
   TemplatePitchMultiplier=1
   TemplateVolumeMultiplier=1
   UE.Source.OnPlay=null
   VolumeModulationRange.Max=1
   VolumeModulationRange.Min=1
   Open.Sounds []
  }
  nodes [
   MSS_Cosmetic_Simple_Base_0 = /FortMetasound/Cosmetics/MSS_Cosmetic_Simple_Base(
    Waves=^Open.Sounds
    PitchModulationRange.Max=^PitchModulationRange.Max
    PitchModulationRange.Min=^PitchModulationRange.Min
    StartDelay.Max=^StartDelay.Max
    StartDelay.Min=^StartDelay.Min
    TemplatePitchMultiplier=^TemplatePitchMultiplier
    TemplateVolumeMultiplier=^TemplateVolumeMultiplier
    UE.Source.OnPlay=^UE.Source.OnPlay
    VolumeModulationRange.Max=^VolumeModulationRange.Max
    VolumeModulationRange.Min=^VolumeModulationRange.Min
   )
   AudioBusWriter1_0 = "UE.Audio Bus Writer (1)"(In 0=UE.Multiply:Audio by Float(PrimaryOperand=UE.Add:Audio(
      PrimaryOperand=@MSS_Cosmetic_Simple_Base_0."UE.OutputFormat.Stereo.Audio:0"
      AdditionalOperands=@MSS_Cosmetic_Simple_Base_0."UE.OutputFormat.Stereo.Audio:1"
     ).Out) {also {defaults {AdditionalOperands=0.15}}}.Out) {also {defaults {Audio Bus=/Game/Sounds/Modulation/EnvelopeFollowers/Volume/AB_Glider_Open_Bus}}}
  ]
  links [
   {dst="@UE.OutputFormat.Stereo.Audio:0", src=MSS_Cosmetic_Simple_Base_0.UE.OutputFormat.Stereo.Audio:0}
   {dst="@UE.OutputFormat.Stereo.Audio:1", src=MSS_Cosmetic_Simple_Base_0.UE.OutputFormat.Stereo.Audio:1}
   {dst="@UE.Source.OneShot.OnFinished", src=MSS_Cosmetic_Simple_Base_0.UE.Source.OneShot.OnFinished}
  ]
 }]
properties {
 assetImportData=&self:AssetImportData
 attenuationSettings=/Game/Sounds/Attenuation/Fort_Player_Glider_OpenClose_Athena_Att
 audioPropertiesSheet=/BRCosmetics/Sounds/Settings/PropertySheets/Glider/APS_Cosmetic_Glider_Open
 maxDistance=10000
 numChannels=2
 outputFormat=Stereo
 soundClassObject=&Fort_Gliders_Land
 concurrencySet [&Athena_GliderOpenClose_Conc]
}
