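# Looping stereo MetaSound source for glider thrust audio, driven by Fort.Glider movement values, with empty default sound arrays.
# Four Glider_ThrustLoop_Direction instances handle forward, backward, sideways and any-direction thrust, each taking loop and one-shot sounds, thresholds and pitch/volume ranges.
# A four-channel stereo mixer sums them with per-layer gains, though all four share Thrust.OneShotVolume.
# The glider-open envelope follower control bus also modulates output volume.
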MetaSoundSource /FortMetasound/Cosmetics/Glider/MSS_Glider_Thrust_Loop
inputs [
 Backward.Loop.Sounds {type=WaveAsset:Array}
 Backward.Modulator.PitchMax {type=Float}
 Backward.Modulator.PitchMin {type=Float}
 Backward.Modulator.VolumeMax {type=Float}
 Backward.Modulator.VolumeMin {type=Float}
 Backward.Sounds {type=WaveAsset:Array}
 Backward.Volume {type=Float}
 Backwards.OneshotThreashold {type=Float}
 "Backwards.ReverseOneshotThreashold " {type=Float}
 Backwards.ReverseOsTrigger {type=Bool}
 Forward.Loop.Sounds {type=WaveAsset:Array}
 Forward.Modulator.PitchMax {type=Float}
 Forward.Modulator.PitchMin {type=Float}
 Forward.Modulator.VolumeMax {type=Float}
 Forward.Modulator.VolumeMin {type=Float}
 Forward.OneshotThreshold {type=Float}
 Forward.ReverseOneShotThreshold {type=Float}
 Forward.Sounds {type=WaveAsset:Array}
 Forward.Volume {type=Float}
 Forwards.ReverseOsTrigger {type=Bool}
 Glider.AnyDirection {type=Float}
 Glider.Backwards {type=Float}
 Glider.Forwards {type=Float}
 Glider.Sideways {type=Float}
 Sideways.Loop.Sounds {type=WaveAsset:Array}
 Sideways.Modulator.PitchMax {type=Float}
 Sideways.Modulator.PitchMin {type=Float}
 Sideways.Modulator.VolumeMax {type=Float}
 Sideways.Modulator.VolumeMin {type=Float}
 Sideways.OneshotThreshold {type=Float}
 Sideways.ReverseOneShotThreshold {type=Float}
 Sideways.ReverseOsTrigger {type=Bool}
 Sideways.Sounds {type=WaveAsset:Array}
 Sideways.Volume {type=Float}
 Thrust.Loop.Sounds {type=WaveAsset:Array}
 Thrust.Modulator.OneShotThreshold {type=Float}
 Thrust.Modulator.PitchMax {type=Float}
 Thrust.Modulator.PitchMin {type=Float}
 Thrust.Modulator.VolumeMax {type=Float}
 Thrust.Modulator.VolumeMin {type=Float}
 Thrust.OneShotThreshold {type=Float}
 Thrust.OneShotVolume {type=Float}
 Thrust.ReverseOneShotThreshold {type=Float}
 Thrust.ReverseOsTrigger {type=Bool}
 Thrust.Sounds {type=WaveAsset:Array}
 Thrust.Volume {type=Float}
 UE.Source.OnPlay {type=Trigger}
]
interfaces [Fort.Glider:1.0, UE.OutputFormat.Stereo:1.0, UE.Source:1.1]
outputs [
 "UE.OutputFormat.Stereo.Audio:0" {type=Audio}
 "UE.OutputFormat.Stereo.Audio:1" {type=Audio}
]
pages [Default {
  input_defaults {
   Backward.Modulator.PitchMax=1.2
   Backward.Modulator.PitchMin=0.8
   Backward.Modulator.VolumeMax=1
   Backward.Modulator.VolumeMin=0.2
   Backward.Volume=1
   Backwards.OneshotThreashold=0.5
   "Backwards.ReverseOneshotThreashold "=0.2
   Backwards.ReverseOsTrigger=null
   Forward.Modulator.PitchMax=1.2
   Forward.Modulator.PitchMin=0.8
   Forward.Modulator.VolumeMax=1
   Forward.Modulator.VolumeMin=0.2
   Forward.OneshotThreshold=0.5
   Forward.ReverseOneShotThreshold=0.2
   Forward.Volume=1
   Forwards.ReverseOsTrigger=null
   Glider.AnyDirection=0
   Glider.Backwards=0
   Glider.Forwards=0
   Glider.Sideways=0
   Sideways.Modulator.PitchMax=1.2
   Sideways.Modulator.PitchMin=0.8
   Sideways.Modulator.VolumeMax=1
   Sideways.Modulator.VolumeMin=0.2
   Sideways.OneshotThreshold=0.5
   Sideways.ReverseOneShotThreshold=0.2
   Sideways.ReverseOsTrigger=null
   Sideways.Volume=1
   Thrust.Modulator.OneShotThreshold=0.5
   Thrust.Modulator.PitchMax=1.2
   Thrust.Modulator.PitchMin=0.8
   Thrust.Modulator.VolumeMax=1
   Thrust.Modulator.VolumeMin=0.2
   Thrust.OneShotThreshold=0.5
   Thrust.OneShotVolume=1
   Thrust.ReverseOneShotThreshold=0.2
   Thrust.ReverseOsTrigger=null
   Thrust.Volume=1
   UE.Source.OnPlay=null
   Backward.Loop.Sounds []
   Backward.Sounds []
   Forward.Loop.Sounds []
   Forward.Sounds []
   Sideways.Loop.Sounds []
   Sideways.Sounds []
   Thrust.Loop.Sounds []
   Thrust.Sounds []
  }
  nodes [
   n1 = &/Glider_ThrustLoop_Direction(
    Value=^Glider.AnyDirection
    Loop=^Thrust.Loop.Sounds
    PitchMax=^Thrust.Modulator.PitchMax
    PitchMin=^Thrust.Modulator.PitchMin
    VolumeMax=^Thrust.Modulator.VolumeMax
    VolumeMin=^Thrust.Modulator.VolumeMin
    OneShotThreshold=^Thrust.OneShotThreshold
    OneShotVolume=^Thrust.OneShotVolume
    "ReverseOneShotThreshold "=^Thrust.ReverseOneShotThreshold
    bIsBackwards=^Thrust.ReverseOsTrigger
    Oneshots=^Thrust.Sounds
    On Play=^UE.Source.OnPlay
   )
   n2 = &Glider_ThrustLoop_Direction(
    Loop=^Backward.Loop.Sounds
    PitchMax=^Backward.Modulator.PitchMax
    PitchMin=^Backward.Modulator.PitchMin
    VolumeMax=^Backward.Modulator.VolumeMax
    VolumeMin=^Backward.Modulator.VolumeMin
    Oneshots=^Backward.Sounds
    OneShotThreshold=^Backwards.OneshotThreashold
    "ReverseOneShotThreshold "=^"Backwards.ReverseOneshotThreashold "
    bIsBackwards=^Backwards.ReverseOsTrigger
    Value=^Glider.Backwards
    OneShotVolume=^Thrust.OneShotVolume
    On Play=^UE.Source.OnPlay
   )
   n3 = &Glider_ThrustLoop_Direction(
    Value=^Glider.Sideways
    Loop=^Sideways.Loop.Sounds
    PitchMax=^Sideways.Modulator.PitchMax
    PitchMin=^Sideways.Modulator.PitchMin
    VolumeMax=^Sideways.Modulator.VolumeMax
    VolumeMin=^Sideways.Modulator.VolumeMin
    OneShotThreshold=^Sideways.OneshotThreshold
    "ReverseOneShotThreshold "=^Sideways.ReverseOneShotThreshold
    bIsBackwards=^Sideways.ReverseOsTrigger
    Oneshots=^Sideways.Sounds
    OneShotVolume=^Thrust.OneShotVolume
    On Play=^UE.Source.OnPlay
   )
   n4 = &Glider_ThrustLoop_Direction(
    Loop=^Forward.Loop.Sounds
    PitchMax=^Forward.Modulator.PitchMax
    PitchMin=^Forward.Modulator.PitchMin
    VolumeMax=^Forward.Modulator.VolumeMax
    VolumeMin=^Forward.Modulator.VolumeMin
    OneShotThreshold=^Forward.OneshotThreshold
    "ReverseOneShotThreshold "=^Forward.ReverseOneShotThreshold
    Oneshots=^Forward.Sounds
    bIsBackwards=^Forwards.ReverseOsTrigger
    Value=^Glider.Forwards
    OneShotVolume=^Thrust.OneShotVolume
    On Play=^UE.Source.OnPlay
   )
   AudioMixerAudioMixerStereo4_0 = "AudioMixer.Audio Mixer (Stereo, 4)"(
    Gain 2=^Backward.Volume
    Gain 0=^Forward.Volume
    Gain 1=^Sideways.Volume
    Gain 3=^Thrust.Volume
    In 3 L=@n1.Out L
    In 3 R=@n1.Out R
    In 2 L=@n2.Out L
    In 2 R=@n2.Out R
    In 1 L=@n3.Out L
    In 1 R=@n3.Out R
    In 0 L=@n4.Out L
    In 0 R=@n4.Out R
   )
  ]
  links [
   {dst="@UE.OutputFormat.Stereo.Audio:0", src=AudioMixerAudioMixerStereo4_0.Out L}
   {dst="@UE.OutputFormat.Stereo.Audio:1", src=AudioMixerAudioMixerStereo4_0.Out R}
  ]
 }]
properties {
 assetImportData=&self:AssetImportData
 attenuationSettings=&Fort_Player_Glider_Thrust_Athena_Att
 audioPropertiesSheet=&APS_Cosmetic_Glider_Loop
 bLooping=true
 duration=10000
 maxDistance=12500
 numChannels=2
 outputFormat=Stereo
 sampleRate=48000
 soundClassObject=&Fort_Gliders_Land
 concurrencySet [&Athena_Glider_Thrust_Conc]
 modulationSettings {
  volumeRouting=Union
  volumeModulationDestination {bEnableModulation=true, modulators [/Game/Sounds/Modulation/EnvelopeFollowers/Volume/CB_Glider_Open_EnvFollower]}
 }
}
